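# Niagara system for the weapon trail of the ServeStreet pickaxe, drawn as one short-lived ribbon emitter.
# Shared FN_AnimTrail modules handle spawning, bone-index setup and updates, with a ribbon ID offset utility at emitter spawn.
# Initial color follows time of day through FNTimeOfDayColor.
# Over particle life, curves taper width and fade alpha, with alpha also multiplied by the system-level SwingFade value.
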NiagaraSystem /BRCosmetics/Weapons/FORT_Melee/Pickaxe_ServeStreet/FX/NS_Pickaxe_ServeStreet_Trail
dependencies {
 usedDataInterfaces [&NiagaraDataInterfaceCurve, &NiagaraDataInterfaceSocketReader, &NiagaraDataInterfaceColorCurve]
 usedDynamicInputs [&ColorFromCurve, &MergeHelper_FloatSwitch]
 usedModules [
  &self:SystemSpawnScript.SystemScriptSource.SystemScriptGraph.NiagaraNodeAssignment_1.SetVariables_#1
  &SystemState
  &self:SystemSpawnScript.SystemScriptSource.SystemScriptGraph.NiagaraNodeAssignment_0.SetVariables_#2
  &self:SystemSpawnScript.SystemScriptSource.SystemScriptGraph.NiagaraNodeAssignment_3.SetVariables_#3
 ]
}
emitters {FN_AnimTrail003 {
  emitterSpawnScript [&FN_RibbonIDOffsetUtility()]
  emitterUpdateScript [
   &EmitterState()
   &FN_SpawnAnimTrail(Spawn Interval=0)
  ]
  particleSpawnScript [
   &FN_InitializeAnimTrailBoneIndecies()
   &FN_InitializeAnimTrail(
    AnimTrailWidth=1.1
    Color=&FNTimeOfDayColor(
     Night Color=<rgba 0.6258 0.747 1 0.5>
     Morning Color=<rgba 1 0.8969 0.8745 0.5>
     Day Color=<rgba 1 1 1 0.9>
     Evening Color=<rgba 1 0.8676 0.806 0.5>
    )
    Lifetime=0.1
   )
  ]
  particleUpdateScript [
   &SolveForcesAndVelocity(Clamp Velocity=false, Limit Acceleration=false)
   &FN_UpdateAnimTrail() {module=FN_UpdateAnimTrail001}
   &ScaleColor(
    Scale Alpha=&Multiply_Float(
     A=0.4
     B=&Multiply_Float(
      A=&FloatFromCurve(FloatCurve=DataInterface() {propertyValues {Curve {
          defaultValue=1
          keys [
           {arriveTangent=-0.2566, interpMode=RCIM_Cubic, leaveTangent=-0.2566, tangentMode=RCTM_User, time=0.1408, value=0.9847}
           {arriveTangent=-2.728, interpMode=RCIM_Cubic, leaveTangent=-2.728, tangentMode=RCTM_User, time=1, value=0}
          ]
         }}})
      B=$System.SwingFade
     )
    )
    ScaleA=true
    ScaleRGB=true
   ) {module=ScaleColor001}
   &ScaleRibbonWidth(Ribbon Width Scale=&FloatFromCurve(FloatCurve=DataInterface() {propertyValues {Curve {
        defaultValue=1
        keys [
         {interpMode=RCIM_Cubic, tangentMode=RCTM_User, time=0, value=1}
         {arriveTangent=-2.3784, interpMode=RCIM_Cubic, leaveTangent=-2.3784, tangentMode=RCTM_User, time=1, value=0}
        ]
       }}}))
  ]
  rendererClasses [&NiagaraRibbonRendererProperties]
  renderers [{rendererClass=&NiagaraRibbonRendererProperties}]
 }}
summary {
 systemName=NS_Pickaxe_ServeStreet_Trail
 emitters [{emitterName=FN_AnimTrail003, rendererClasses [&NiagaraRibbonRendererProperties]}]
 userVariables [
  User.bIsFrontEnd {type=NiagaraBool}
  User.bIsFrontEndPreview {type=NiagaraBool}
  User.PhysicalSurface {type=NiagaraInt32}
  User.RibbonIDOffset {type=NiagaraInt32}
  User.SwingSectionIndex {type=NiagaraInt32}
 ]
}
